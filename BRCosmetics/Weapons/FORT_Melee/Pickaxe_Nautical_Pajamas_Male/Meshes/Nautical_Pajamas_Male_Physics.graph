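# Physics asset for the Nautical Pajamas (male) pickaxe cosmetic mesh, defining simulated dangly parts on the two-handed melee weapon.
# Capsule bodies cover the handle, left and right two-bone eye and wing chains, a stuffing piece and a zipper.
# Angle-limited constraints tie the eye and wing roots, stuffing and zipper to the handle, and each second bone to its root.
# Wing roots are stiff with looser outer segments, while the zipper swings widest.
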PhysicsAsset /BRCosmetics/Weapons/FORT_Melee/Pickaxe_Nautical_Pajamas_Male/Meshes/Nautical_Pajamas_Male_Physics
bodies {
 Melee_TwoHanded_Handle [{shapeType=Capsule, length=90.074, radius=2.9951, center=<0.8833 -1.0219 16.7287>, rotation=<rot 0.0488 -89.9641 -0.1314>}]
 dyn_eye_l_01 [{shapeType=Capsule, length=1, radius=2, center=<2 0 0>}]
 dyn_eye_l_02 [{shapeType=Capsule, length=0, radius=3}]
 dyn_eye_r_01 [{shapeType=Capsule, length=1, radius=2, center=<2.0013 0 0>}]
 dyn_eye_r_02 [{shapeType=Capsule, length=0, radius=3}]
 dyn_stuffing [{shapeType=Capsule, length=0, radius=5.5265, center=<-1.1961 0.1361 4.3115>, rotation=<rot 0 84.1458 -78.888>}]
 dyn_wing_l_01 [{shapeType=Capsule, length=8, radius=2, rotation=<rot -4.0784 -90 82.9589>}]
 dyn_wing_l_02 [{shapeType=Capsule, length=9.6955, radius=6.0281, center=<2.3136 0.8603 -6.0347>, rotation=<rot -9.8937 -89.9995 -28.9639>}]
 dyn_wing_r_01 [{shapeType=Capsule, length=8, radius=2, rotation=<rot 2.7624 -89.9995 -78.6456>}]
 dyn_wing_r_02 [{shapeType=Capsule, length=9.4073, radius=5.5841, center=<1.7224 -0.6612 -6.1562>, rotation=<rot 9.0949 -89.9995 -24.1952>}]
 dyn_zipper [{shapeType=Capsule, length=3.8289, radius=1.4142, center=<-0.6887 0.0503 -2.1641>, rotation=<rot 0.9019 -89.9995 10.5891>}]
}
constraints [
 {
  bone1Name=dyn_eye_l_01
  bone2Name=Melee_TwoHanded_Handle
  swing1LimitDegrees=10
  swing1Motion=Locked
  swing2LimitDegrees=10
  twistLimitDegrees=10
 }
 {bone1Name=dyn_eye_l_02, bone2Name=dyn_eye_l_01, swing1LimitDegrees=10, swing2LimitDegrees=2, twistLimitDegrees=10}
 {bone1Name=dyn_eye_r_01, bone2Name=Melee_TwoHanded_Handle, swing1LimitDegrees=1, swing2LimitDegrees=10, twistLimitDegrees=10}
 {bone1Name=dyn_eye_r_02, bone2Name=dyn_eye_r_01, swing1LimitDegrees=10, swing2LimitDegrees=2, twistLimitDegrees=10}
 {bone1Name=dyn_wing_l_01, bone2Name=Melee_TwoHanded_Handle, swing1LimitDegrees=5, swing2LimitDegrees=5, twistLimitDegrees=0}
 {bone1Name=dyn_wing_l_02, bone2Name=dyn_wing_l_01, swing1LimitDegrees=45, swing2LimitDegrees=45, twistLimitDegrees=0}
 {bone1Name=dyn_wing_r_01, bone2Name=Melee_TwoHanded_Handle, swing1LimitDegrees=5, swing2LimitDegrees=5, twistLimitDegrees=0}
 {bone1Name=dyn_wing_r_02, bone2Name=dyn_wing_r_01, swing1LimitDegrees=45, swing2LimitDegrees=45, twistLimitDegrees=0}
 {bone1Name=dyn_stuffing, bone2Name=Melee_TwoHanded_Handle, swing1LimitDegrees=38, swing2LimitDegrees=17, twistLimitDegrees=8}
 {bone1Name=dyn_zipper, bone2Name=Melee_TwoHanded_Handle, swing1LimitDegrees=20, swing2LimitDegrees=90, twistLimitDegrees=5}
]
